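# Translucent, unlit, two-sided debug material for Niagara particles in the TroutWrist skydiving contrail effect.
# Squared parabolic masks from U and V drive SplineThicken, with dynamic parameters setting base and tip width.
# The U mask also shapes particle color plus a warm tint, scaled by dynamic EmissiveMult into MF_FX_EmissiveAdjustments.
# Sweep, panner and scalar EmissiveMult nodes feed no other node shown, suggesting leftover experiments.
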Material /BRCosmetics/Athena/Effects/Particles/Skydiving/NS_Contrail_TroutWrist/Materials/M_Debug_Sweep
graph {
 properties {
  bAutomaticallySetUsageInEditor=false
  bUsedWithNiagaraMeshParticles=true
  bUsedWithNiagaraRibbons=true
  bUsedWithNiagaraSprites=true
  blendMode=BLEND_Translucent
  opacityMaskClipValue=0.01
  shadingModel=MSM_Unlit
  twoSided=true
  usedShadingModels=Unlit
 }
 nodes [
  n1 = TextureCoordinate()
  n2 = ComponentMask(None=@n1) {r=true}
  n3 = Multiply(A=OneMinus(None=@n2), B=@n2)
  n4 = Multiply(A=@n3, B=@n3)
  n5 = DynamicParameter() {defaultValue=<rgba 15 15 1 1>, paramNames [WidthBase,WidthTip,Sweep,EmissiveMult]}
  n6 = ComponentMask(None=@n1) {g=true}
  n7 = Multiply(A=OneMinus(None=@n6), B=@n6)
  n8 = Divide(A=Multiply(A=@n7, B=@n7)) {constB=0.1}
  MaterialFunctionCall(
   WidthBase=Multiply(A=@n4, B=@n5.WidthBase)
   WidthTip=@n5.WidthTip
   WorldPosition=WorldPosition().XYZ
   UVs for Thickness=OneMinus(None=@n8)
   FlattenPixelNormal=StaticBool() {value=true}
  ) {materialFunction=&SplineThicken}
  n9 = Divide(A=@n4) {constB=0.1}
  n10 = Multiply(A=ParticleColor().RGB, B=@n9)
  n11 = Power(Base=@n9) {constExponent=0.1}
  n12 = Multiply(A=@n11, B=Constant3Vector() {constant=<rgba 1 0.7638 0.2378 1>})
  MaterialFunctionCall(Emissive=Multiply(A=Add(A=@n10, B=@n12), B=@n5.EmissiveMult)) {materialFunction=&MF_FX_EmissiveAdjustments}
  Multiply(B=Saturate(None=@n8)) {constA=1}
  OneMinus(None=@n11)
  Panner(Coordinate=@n1, Time=Time())
  Power(Base=ComponentMask(None=Subtract(A=@n5.Sweep, B=@n1)) {g=true})
  ScalarParameter() {defaultValue=1, parameterName=EmissiveMult}
 ]
}
