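# Niagara skydiving contrail cosmetic that tosses tumbling fruit meshes upward, then shows them being sliced with colored swipes.
# Em_Mesh_Fruit spawns four partitioned variants; sliced-piece and swipe emitters sample its particles within an age window, which shifts with glide lean.
# A separate banana emitter runs alongside, while banana drip and splash emitters are disabled.
# Lean also drives spawn rate and speed, front-end flag adjusts spread, and camera distance scales meshes.
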NiagaraSystem /BRCosmetics/Athena/Effects/Particles/Skydiving/NS_Contrail_TroutWrist/NS_Contrail_TroutWrist
dependencies {
 usedDataInterfaces [
  &NiagaraDataInterfaceMeshRendererInfo
  &NiagaraDataInterfaceCurve
  &NiagaraDataInterfaceCamera
  &NiagaraDataInterfaceParticleRead
  &NiagaraDataInterfaceArrayColor
  &NiagaraDataInterfaceSpriteRendererInfo
  &NiagaraDataInterfaceArrayFloat3
  &NiagaraDataInterfaceColorCurve
  &NiagaraDataInterfaceVector2DCurve
  &NiagaraDataInterfaceSkeletalMesh
 ]
 usedDynamicInputs [&FNTimeOfDayScalar]
 usedModules [
  &self:SystemSpawnScript.SystemScriptSource.SystemScriptGraph.NiagaraNodeAssignment_0.SetVariables_#1
  &self:SystemSpawnScript.SystemScriptSource.SystemScriptGraph.NiagaraNodeAssignment_3.SetVariables_#2
  &self:SystemSpawnScript.SystemScriptSource.SystemScriptGraph.NiagaraNodeAssignment_1.SetVariables_#3
  &SystemState
  &self:SystemSpawnScript.SystemScriptSource.SystemScriptGraph.NiagaraNodeAssignment_2.SetVariables_#4
 ]
}
emitters {
 Em_Banana_Drips {
  bEnabled=false
  emitterUpdateScript [
   &EmitterState()
   &SpawnBurst_Instantaneous(Spawn Count=24, Use Spawn Probability=false, UseLoopCountLimit=false) {enabled=false}
   &SpawnParticlesFromOtherEmitter(
    Attribute Reader=DataInterface() {propertyValues {EmitterBinding {bindingMode=Other, emitterName=Em_Mesh_Fruit}}}
    Calculate Spawn Rate Per Particle=false
    Spawn Rate=256
    Use Maximum Age=false
    Use Minimum Age=false
   )
  ]
  particleSpawnScript [
   &self:ScratchModule(Particle Attribute Reader=DataInterface() {propertyValues {EmitterBinding {bindingMode=Other, emitterName=Em_Mesh_Fruit}}}) {module=ScratchModule}
   &InitializeParticle(
    Color=<rgba 1 0.7029 0.2944 1>
    Lifetime Max=0.6
    Lifetime Min=0.4
    Lifetime Mode=::NewEnumerator1
    Sprite Rotation Mode=::NewEnumerator2
    Sprite Size Mode=::NewEnumerator3
    Uniform Sprite Size=&RandomRangeFloat(Minimum=2, Maximum=4)
    UsePositionOffset=false
    Write Material Random=true
   )
   &SubUVAnimation(
    End Frame Range Override=3
    Evaluation Type=::NewEnumerator0
    "Sprite Renderer "=DataInterface() {propertyValues {SpriteRenderer=&self:Em_Banana_Drips_7.NiagaraSpriteRendererProperties_1}}
    SubUV Animation Mode:ENiagara_SubUVLookupModeV2=::NewEnumerator2
    UseEndFrame=false
    UseStartFrame=false
   )
   &AddVelocity(
    Module.Use Falloff Distance=false
    Override Seed=false
    Velocity Mode=::NewEnumerator1
    Velocity Speed=&RandomRangeFloat(Minimum=100, Maximum=250)
   )
   &SampleParticlesFromOtherEmitter(
    Maximum Sampled Particle Age=1.625
    Minimum Sampled Particle Age=1.4
    Module.MaxSampledAgeCutoff=true
    Module.MinSampledAgeCutoff=true
    Reject Particle By Sampled Age=true
   )
  ]
  particleUpdateScript [
   &ParticleState()
   &Drag(Use Linear Drag=true, Use Rotational Drag=true)
   &GravityForce() {enabled=false}
   &SolveForcesAndVelocity(Clamp Velocity=false, Limit Acceleration=false)
   &ScaleSpriteSizeBySpeed(Max Scale Factor=<2 5>)
   &ScaleColor(
    Scale Alpha=&FloatFromCurve(FloatCurve=DataInterface() {propertyValues {Curve {keys [{interpMode=RCIM_Cubic, tangentMode=RCTM_User, time=0, value=1}, {interpMode=RCIM_Cubic, tangentMode=RCTM_User, time=1, value=0}]}}})
    ScaleA=true
    ScaleRGB=false
   ) {enabled=false}
   &ScaleSpriteSize(Uniform Curve Sprite Scale=DataInterface() {propertyValues {Curve {keys [
        {interpMode=RCIM_Cubic, tangentMode=RCTM_User, time=0, value=1}
        {arriveTangent=-2.5389, interpMode=RCIM_Cubic, leaveTangent=-2.5389, tangentMode=RCTM_User, time=1, value=0}
       ]}}})
   &ScaleColor(
    Scale Alpha=&Multiply_Float(
     A=&FloatFromCurve(FloatCurve=DataInterface() {propertyValues {Curve {keys [<at 0 1 RCIM_Cubic>,<at 1 0>]}}})
     B=1
    )
    Scale RGB=&MakeVectorFromLinearColorRGB(Color=&MultiplyLinearColor(
      Linear Color=&ColorFromCurve(ColorCurve=DataInterface() {propertyValues {
         AlphaCurve {keys [<at 0 1 RCIM_Cubic>,<at 1 0 RCIM_Cubic>]}
         BlueCurve {keys [<at 0 3.8266 RCIM_Cubic>,<at 0.2095 0.3827>,<at 1 1 RCIM_Cubic>]}
         GreenCurve {keys [<at 0 7.6631 RCIM_Cubic>,<at 0.2095 0.7663>,<at 1 1 RCIM_Cubic>]}
         RedCurve {keys [<at 0 10 RCIM_Cubic>,<at 0.2095 1>,<at 1 1 RCIM_Cubic>]}
        }})
      Scale Factor=&MakeVector4_FromLinearColor(LinearColor=$System.ToD Color)
     ))
    ScaleA=true
    ScaleRGB=true
   ) {enabled=false, module=ScaleColor001}
   &ScaleColor(
    Scale Alpha=&Multiply_Float(
     A=&FloatFromCurve(FloatCurve=DataInterface() {propertyValues {Curve {keys [<at 0 1 RCIM_Cubic>,<at 1 0>]}}})
     B=1
    )
    Scale RGB=&MakeVectorFromLinearColorRGB(Color=&MultiplyLinearColor(
      Linear Color=<rgba 1 0.8622 0.5003 1>
      Scale Factor=&MakeVector4_FromLinearColor(LinearColor=$System.ToD Color)
     ))
    ScaleA=true
    ScaleRGB=true
   ) {module=ScaleColor002}
  ]
  rendererClasses [&NiagaraSpriteRendererProperties]
  renderers [{rendererClass=&NiagaraSpriteRendererProperties}]
 }
 Em_Banana_Splash {
  bEnabled=false
  emitterUpdateScript [
   &EmitterState()
   &SpawnBurst_Instantaneous(Spawn Count=3, Use Spawn Probability=false, UseLoopCountLimit=false) {enabled=false}
   &SpawnParticlesFromOtherEmitter(
    Attribute Reader=DataInterface() {propertyValues {EmitterBinding {bindingMode=Other, emitterName=Em_Mesh_Fruit}}}
    Spawn Rate=16
    Spawn Rate Per Particle Cap=64
    Use Maximum Age=false
    Use Minimum Age=false
   )
  ]
  particleSpawnScript [
   &self:ScratchModule(Particle Attribute Reader=DataInterface() {propertyValues {EmitterBinding {bindingMode=Other, emitterName=Em_Mesh_Fruit}}}) {module=ScratchModule}
   &InitializeParticle(
    Color=<rgba 0.936 0.7796 0.5646 1>
    Lifetime Max=0.5
    Lifetime Min=0.4
    Lifetime Mode=::NewEnumerator1
    Sprite Rotation Mode=::NewEnumerator2
    Sprite Size Mode=::NewEnumerator1
    Uniform Sprite Size Max=75
    Uniform Sprite Size Min=65
    UsePositionOffset=false
    Write Material Random=true
   )
   &SampleParticlesFromOtherEmitter(
    Maximum Sampled Particle Age=1.625
    Minimum Sampled Particle Age=1.55
    Module.MaxSampledAgeCutoff=true
    Module.MinSampledAgeCutoff=true
    Reject Particle By Sampled Age=true
   )
   &SubUVAnimation(
    End Frame Range Override=3
    Evaluation Type=::NewEnumerator0
    "Sprite Renderer "=DataInterface() {propertyValues {SpriteRenderer=&self:Em_Banana_Splash_9.NiagaraSpriteRendererProperties_1}}
    SubUV Animation Mode:ENiagara_SubUVLookupModeV2=::NewEnumerator2
    UseEndFrame=false
    UseStartFrame=false
   )
   &ShapeLocation(
    Hemisphere Distribution=<1 0.7>
    MaskSpawnGroup=true
    Offset Mode=::NewEnumerator0
    Override Seed=false
    Sphere Radius=1
    Sphere Surface Distribution=0.15
    UseOwnerScale=true
   )
   &AddVelocity(Velocity=&Multiply_Vector(
     A=&UniformRangedVector(Minimum=<0 100 -200>, Maximum=<0 1000 -200>)
     B=&MakeCustomVectorFromBool(
      Bool=$System.LastSwing
      True Vector=&SelectVectorFromArray(
       Array Sampling Mode=::NewEnumerator0
       Direct Array Index=?
       Vector Selection Array=DataInterface() {propertyValues {FloatData [<1 1 1>, <1 -1 1>]}}
       Random Seed=0
      )
      False Vector=$System.DirectionVector
     )
    )) {module=AddVelocity002}
   &CameraOffset(Camera Offset Amount=20)
  ]
  particleUpdateScript [
   &ParticleState()
   &Drag(Drag=2, Use Linear Drag=true, Use Rotational Drag=false)
   &GravityForce() {enabled=false}
   &InheritSourceMovement(
    Applied Force Scale=<1 1 -1>
    Applied Force Speed Limit=15
    ApplyForce=true
    ApplyForceFalloff=false
    ApplyPosition=false
    Module.LimitForceSpeed=true
   ) {enabled=false}
   &SolveForcesAndVelocity(Clamp Velocity=false, Limit Acceleration=false)
   &ScaleSpriteSizeBySpeed(Max Scale Factor=<1 4>)
   &ScaleColor(
    Scale Alpha=&FloatFromCurve(FloatCurve=DataInterface() {propertyValues {Curve {keys [
         {interpMode=RCIM_Cubic, tangentMode=RCTM_User, time=0, value=1}
         {arriveTangent=-2.6707, interpMode=RCIM_Cubic, leaveTangent=-2.6707, tangentMode=RCTM_User, time=1, value=0}
        ]}}})
    ScaleA=true
    ScaleRGB=false
   ) {enabled=false}
   &ScaleSpriteSize(
    Uniform Curve Scale=0.5
    Uniform Curve Sprite Scale=DataInterface() {propertyValues {Curve {keys [
        {arriveTangent=1.2875, interpMode=RCIM_Cubic, leaveTangent=1.2875, tangentMode=RCTM_User, time=0, value=1}
        {interpMode=RCIM_Cubic, tangentMode=RCTM_User, time=1, value=2}
       ]}}}
   )
   &DynamicMaterialParameters(
    Module.Param0WriteEnabled1=true
    Module.Param1WriteEnabled1=true
    Module.Param2WriteEnabled1=true
    Module.Param3WriteEnabled1=true
    Write Parameter Index 0=false
    Write Parameter Index 1=true
   )
   &ScaleColor(
    Scale Alpha=&Multiply_Float(
     A=&FloatFromCurve(FloatCurve=DataInterface() {propertyValues {Curve {keys [<at 0 1 RCIM_Cubic>,<at 1 0>]}}})
     B=1
    )
    Scale RGB=&MakeVectorFromLinearColorRGB(Color=&MultiplyLinearColor(
      Linear Color=&ColorFromCurve(ColorCurve=DataInterface() {propertyValues {
         AlphaCurve {keys [<at 0 1 RCIM_Cubic>,<at 1 0 RCIM_Cubic>]}
         BlueCurve {keys [<at 0 3.8266 RCIM_Cubic>,<at 0.245 0.3827>,<at 1 1 RCIM_Cubic>]}
         GreenCurve {keys [<at 0 7.6631 RCIM_Cubic>,<at 0.245 0.7663>,<at 1 1 RCIM_Cubic>]}
         RedCurve {keys [<at 0 10 RCIM_Cubic>,<at 0.245 1>,<at 1 1 RCIM_Cubic>]}
        }})
      Scale Factor=&MakeVector4_FromLinearColor(LinearColor=$System.ToD Color)
     ))
    ScaleA=true
    ScaleRGB=true
   ) {enabled=false, module=ScaleColor001}
   &ScaleColor(
    Scale Alpha=&Multiply_Float(
     A=&FloatFromCurve(FloatCurve=DataInterface() {propertyValues {Curve {keys [<at 0 1 RCIM_Cubic>,<at 1 0>]}}})
     B=1
    )
    Scale RGB=&MakeVectorFromLinearColorRGB(Color=&MultiplyLinearColor(
      Linear Color=<rgba 1 0.8622 0.5003 1>
      Scale Factor=&MakeVector4_FromLinearColor(LinearColor=$System.ToD Color)
     ))
    ScaleA=true
    ScaleRGB=true
   ) {module=ScaleColor002}
  ]
  rendererClasses [&NiagaraSpriteRendererProperties]
  renderers [{rendererClass=&NiagaraSpriteRendererProperties}]
 }
 Em_Mesh_Fruit {
  emitterUpdateScript [
   &EmitterState()
   &SpawnRate(SpawnRate=&Lerp_Float(A=5, B=10, Alpha=$User.ParaGlideLeanAlpha), Use Spawn Probability=false)
  ]
  particleSpawnScript [
   &PartitionParticles(Number Of Partitions=4)
   /SetVariables(Particles.VisibilityTag=$Particles.PartitionParticles.Partition) {isSetParameters=true, module=SetVariables_#5}
   &InitializeParticle(
    Lifetime=$System.Fruit_Lifetime
    Mesh Scale Mode=::NewEnumerator3
    Mesh Uniform Scale=0.75
    UsePositionOffset=false
    Write Material Random=true
   )
   &ShapeLocation(
    MaskSpawnGroup=false
    Non Uniform Scale=<1 1 0.2>
    Offset=&MakeVector(X=0, Y=35, Z=0)
    Offset Mode=::NewEnumerator0
    Override Seed=false
    Ring Radius=&MakeCustomFloatFromBool(Bool=$User.bIsFrontEnd, True Float=25, False Float=65)
    Rotation Coordinate Space=::Simulation
    Shape Primitive=::NewEnumerator4
    UseOwnerScale=true
   )
   &AddVelocity(
    Cone Angle=&MakeCustomFloatFromBool(Bool=$User.bIsFrontEnd, True Float=8, False Float=15)
    Cone Axis=<0 0 1>
    Module.UseConeFalloff=false
    Module.UseVelDistribution=false
    Override Seed=false
    Velocity Mode=::NewEnumerator2
    Velocity Speed=&Lerp_Float(
     A=&RandomRangeFloat(Minimum=275, Maximum=425)
     B=&MakeCustomFloatFromBool(Bool=$User.bIsFrontEnd, True Float=750, False Float=500)
     Alpha=$User.ParaGlideLeanAlpha
    )
   )
   &InitialMeshOrientation(Use Rotation Vector=false)
  ]
  particleUpdateScript [
   &ParticleState()
   &ScaleMeshSize(Scale Factor=&VectorFromFloat(Value=&FloatFromCurve(
      FloatCurve=DataInterface() {propertyValues {Curve {keys [
          {arriveTangent=-0.0357, interpMode=RCIM_Cubic, leaveTangent=-0.0357, tangentMode=RCTM_User, time=0, value=1}
          {arriveTangent=-14.6213, leaveTangent=-2, tangentMode=RCTM_User, time=0.05, value=0}
          <at 0.5 0>
          {arriveTangent=13.566, interpMode=RCIM_Cubic, leaveTangent=13.566, tangentMode=RCTM_User, time=0.95, value=0}
          {arriveTangent=0.1498, tangentMode=RCTM_User, time=1, value=1}
         ]}}}
      CurveIndex=&NormalizeDistanceRange(
       Start Position:NiagaraPosition=&ReturnCameraPosition(Coordinate Space:ENiagaraCoordinateSpace=::Simulation, Camera Query:NiagaraDataInterfaceCamera=DataInterface())
       End Position:NiagaraPosition=$Particles.Position
       Distance:NiagaraFloat=100
       Invert Normalized Range:NiagaraBool=false
      )
     ))) {module=ScaleMeshSize001}
   &ScaleMeshSize(Scale Factor=&VectorFromFloat(Value=&FloatFromCurve(FloatCurve=DataInterface() {propertyValues {Curve {keys [{arriveTangent=5.5564, interpMode=RCIM_Cubic, leaveTangent=5.5564, tangentMode=RCTM_User, time=0, value=0}, <at 0.5 1 RCIM_Cubic>]}}})))
   &UpdateMeshOrientation(Rotation Vector=&RandomRangeVector(Minimum=<-1 -1 -1>, Maximum=<1 1 1>))
   &PointAttractionForce(
    Attraction Radius:NiagaraFloat=$System.AttractionRadius
    Attraction Strength=$System.AttractionStrength
    Attractor Position Offset:Vector3f=<0 0 30>
    Falloff Exponent=0.01
    Kill Within Radius=false
    Module.Use Falloff=true
    UseAttractorOffset=true
   ) {module=PointAttractionForce001}
   &GravityForce(Gravity=&MakeVector(X=0, Y=0, Z=&RandomRangeFloat(Minimum=-275, Maximum=-375)))
   &SolveForcesAndVelocity(Clamp Velocity=false, Limit Acceleration=false)
   &self:ScaleMesh_byCamDist(Camera Query:NiagaraDataInterfaceCamera=$Emitter.Camera Query, MaxDist=250, MaxScale=1, MinDist=125) {enabled=false, module=ScaleMesh_byCamDist}
  ]
  rendererClasses [&NiagaraMeshRendererProperties]
  renderers [
   {rendererClass=&NiagaraMeshRendererProperties, rendererIndex=0}
   {rendererClass=&NiagaraMeshRendererProperties, rendererIndex=1}
   {rendererClass=&NiagaraMeshRendererProperties, rendererIndex=2}
   {rendererClass=&NiagaraMeshRendererProperties, rendererIndex=3}
  ]
 }
 Em_Mesh_Fruit_Banana {
  emitterUpdateScript [
   &EmitterState()
   &SpawnRate(SpawnRate=&Lerp_Float(A=8, B=20, Alpha=$User.ParaGlideLeanAlpha), Use Spawn Probability=false)
  ]
  particleSpawnScript [
   &PartitionParticles(Number Of Partitions=4)
   /SetVariables(Particles.VisibilityTag=$Particles.PartitionParticles.Partition) {isSetParameters=true, module=SetVariables_#5}
   &InitializeParticle(Lifetime=1.3, Mesh Scale Mode=::NewEnumerator3, Mesh Uniform Scale=0.75, UsePositionOffset=false, Write Material Random=true)
   &ShapeLocation(
    MaskSpawnGroup=false
    Non Uniform Scale=<1 1 0.2>
    Offset=&MakeVector(X=0, Y=35, Z=0)
    Offset Mode=::NewEnumerator0
    Override Seed=false
    Ring Radius=&MakeCustomFloatFromBool(Bool=$User.bIsFrontEnd, True Float=25, False Float=65)
    Rotation Coordinate Space=::Simulation
    Shape Primitive=::NewEnumerator4
    UseOwnerScale=true
   )
   &AddVelocity(
    Cone Angle=&MakeCustomFloatFromBool(Bool=$User.bIsFrontEnd, True Float=8, False Float=15)
    Cone Axis=<0 0 1>
    Module.UseConeFalloff=false
    Module.UseVelDistribution=false
    Override Seed=false
    Velocity Mode=::NewEnumerator2
    Velocity Speed=&Lerp_Float(A=&RandomRangeFloat(Minimum=275, Maximum=325), B=500, Alpha=$User.ParaGlideLeanAlpha)
   )
   &InitialMeshOrientation(Use Rotation Vector=false)
  ]
  particleUpdateScript [
   &ParticleState()
   &ScaleMeshSize(Scale Factor=&VectorFromFloat(Value=&FloatFromCurve(
      FloatCurve=DataInterface() {propertyValues {Curve {keys [
          {arriveTangent=-0.0357, interpMode=RCIM_Cubic, leaveTangent=-0.0357, tangentMode=RCTM_User, time=0, value=1}
          {arriveTangent=-14.6213, leaveTangent=-2, tangentMode=RCTM_User, time=0.05, value=0}
          <at 0.5 0>
          {arriveTangent=13.566, interpMode=RCIM_Cubic, leaveTangent=13.566, tangentMode=RCTM_User, time=0.95, value=0}
          {arriveTangent=0.1498, tangentMode=RCTM_User, time=1, value=1}
         ]}}}
      CurveIndex=&NormalizeDistanceRange(
       Start Position:NiagaraPosition=&ReturnCameraPosition(Coordinate Space:ENiagaraCoordinateSpace=::Simulation, Camera Query:NiagaraDataInterfaceCamera=DataInterface())
       End Position:NiagaraPosition=$Particles.Position
       Distance:NiagaraFloat=100
       Invert Normalized Range:NiagaraBool=false
      )
     ))) {module=ScaleMeshSize001}
   &ScaleMeshSize(Scale Factor=&VectorFromFloat(Value=&FloatFromCurve(FloatCurve=DataInterface() {propertyValues {Curve {keys [
          {arriveTangent=5.5564, interpMode=RCIM_Cubic, leaveTangent=5.5564, tangentMode=RCTM_User, time=0, value=0}
          {arriveTangent=0.3247, leaveTangent=1.3333, tangentMode=RCTM_User, time=0.5, value=1}
          {arriveTangent=-2, interpMode=RCIM_Cubic, leaveTangent=-2, time=0.75, value=1}
          {interpMode=RCIM_Cubic, leaveTangent=-7.0392, time=1, value=0}
         ]}}})))
   &UpdateMeshOrientation(Rotation Vector=&RandomRangeVector(Minimum=<-1 -1 -1>, Maximum=<1 1 1>))
   &PointAttractionForce(
    Attraction Radius:NiagaraFloat=$System.AttractionRadius
    Attraction Strength=$System.AttractionStrength
    Attractor Position Offset:Vector3f=<0 0 30>
    Falloff Exponent=0.01
    Kill Within Radius=false
    Module.Use Falloff=true
    UseAttractorOffset=true
   )
   &SolveForcesAndVelocity(Clamp Velocity=false, Limit Acceleration=false)
   &self:ScaleMesh_byCamDist(Camera Query:NiagaraDataInterfaceCamera=$Emitter.Camera Query, MaxDist=250, MaxScale=1, MinDist=125) {enabled=false, module=ScaleMesh_byCamDist}
  ]
  rendererClasses [&NiagaraMeshRendererProperties]
  renderers [{rendererClass=&NiagaraMeshRendererProperties, rendererIndex=0}]
 }
 Em_Mesh_Fruit_Sliced {
  emitterUpdateScript [
   &EmitterState()
   &SpawnParticlesFromOtherEmitter(
    Attribute Reader=DataInterface() {propertyValues {EmitterBinding {bindingMode=Other, emitterName=Em_Mesh_Fruit}}}
    Spawn Rate=24
    Spawn Rate Per Particle Cap=64
    Use Maximum Age=false
    Use Minimum Age=false
   )
  ]
  particleSpawnScript [
   &self:ScratchModule(Particle Attribute Reader=DataInterface() {propertyValues {EmitterBinding {bindingMode=Other, emitterName=Em_Mesh_Fruit}}}) {module=ScratchModule}
   /SetVariables(Particles.VisibilityTag=$Particles.Partition) {isSetParameters=true}
   &InitializeParticle(Color Mode=::NewEnumerator0, Mesh Scale Mode=::NewEnumerator3, Mesh Uniform Scale=0.75, UsePositionOffset=false, Write Material Random=true)
   &SampleParticlesFromOtherEmitter(
    Maximum Sampled Particle Age:NiagaraFloat=&Lerp_Float(
     A=1.325
     B=0.425
     Alpha=&FloatFromCurve(
      FloatCurve=DataInterface() {propertyValues {Curve {keys [<at 0 1 RCIM_Constant>,<at 0.1 0 RCIM_Constant>]}}}
      Scale Curve=$User.ParaGlideLeanAlpha
     )
    )
    Minimum Sampled Particle Age:NiagaraFloat=&Lerp_Float(
     A=1.225
     B=0.325
     Alpha=&FloatFromCurve(
      FloatCurve=DataInterface() {propertyValues {Curve {keys [<at 0 1 RCIM_Constant>,<at 0.1 0 RCIM_Constant>]}}}
      Scale Curve=$User.ParaGlideLeanAlpha
     )
    )
    Module.MaxSampledAgeCutoff=true
    Module.MinSampledAgeCutoff=true
    Reject Particle By Sampled Age=true
   )
   &AddVelocity(
    Cone Angle=90
    Cone Axis=<0 0 1>
    Module.UseConeFalloff=false
    Module.UseVelDistribution=false
    Override Seed=false
    Velocity Mode=::NewEnumerator2
    Velocity Speed=&Lerp_Float(A=75, B=125, Alpha=$User.ParaGlideLeanAlpha)
   )
   &InitialMeshOrientation(Use Rotation Vector=false)
  ]
  particleUpdateScript [
   &ParticleState()
   &ScaleMeshSize(Scale Factor=&VectorFromFloat(Value=&FloatFromCurve(
      FloatCurve=DataInterface() {propertyValues {Curve {keys [
          {arriveTangent=-0.0357, interpMode=RCIM_Cubic, leaveTangent=-0.0357, tangentMode=RCTM_User, time=0, value=1}
          {arriveTangent=-14.6213, leaveTangent=-2, tangentMode=RCTM_User, time=0.05, value=0}
          <at 0.5 0>
          {arriveTangent=13.566, interpMode=RCIM_Cubic, leaveTangent=13.566, tangentMode=RCTM_User, time=0.95, value=0}
          {arriveTangent=0.1498, tangentMode=RCTM_User, time=1, value=1}
         ]}}}
      CurveIndex=&NormalizeDistanceRange(
       Start Position:NiagaraPosition=&ReturnCameraPosition(Coordinate Space:ENiagaraCoordinateSpace=::Simulation, Camera Query:NiagaraDataInterfaceCamera=DataInterface())
       End Position:NiagaraPosition=$Particles.Position
       Distance:NiagaraFloat=100
       Invert Normalized Range:NiagaraBool=false
      )
     ))) {module=ScaleMeshSize001}
   &ScaleMeshSize(Scale Factor=&VectorFromFloat(Value=&FloatFromCurve(FloatCurve=DataInterface() {propertyValues {Curve {keys [<at 0.5 1 RCIM_Cubic>, {arriveTangent=-4.2776, tangentMode=RCTM_User, time=1, value=0}]}}})))
   &UpdateMeshOrientation(Rotation Rate=1.5, Rotation Vector=&RandomRangeVector(Minimum=<-1 -1 -1>, Maximum=<1 1 1>))
   &PointAttractionForce(
    Attraction Radius:NiagaraFloat=$System.AttractionRadius
    Attraction Strength=$System.AttractionStrength
    Attractor Position Offset:Vector3f=<0 0 30>
    Falloff Exponent=0.01
    Kill Within Radius=false
    Module.Use Falloff=true
    UseAttractorOffset=true
   ) {module=PointAttractionForce001}
   &GravityForce(Gravity=&MakeVector(X=0, Y=0, Z=&RandomRangeFloat(Minimum=-275, Maximum=-375)))
   &SolveForcesAndVelocity(Clamp Velocity=false, Limit Acceleration=false)
   &CameraOffset(Camera Offset Amount=-15)
   &self:ScaleMesh_byCamDist(Camera Query:NiagaraDataInterfaceCamera=$Emitter.Camera Query, MaxDist=250, MaxScale=1, MinDist=125) {enabled=false, module=ScaleMesh_byCamDist}
  ]
  rendererClasses [&NiagaraMeshRendererProperties]
  renderers [
   {rendererClass=&NiagaraMeshRendererProperties, rendererIndex=0}
   {rendererClass=&NiagaraMeshRendererProperties, rendererIndex=1}
   {rendererClass=&NiagaraMeshRendererProperties, rendererIndex=2}
   {rendererClass=&NiagaraMeshRendererProperties, rendererIndex=3}
  ]
 }
 Em_Mesh_Swipe {
  emitterUpdateScript [
   &EmitterState()
   &SpawnParticlesFromOtherEmitter(
    Attribute Reader=DataInterface() {propertyValues {EmitterBinding {bindingMode=Other, emitterName=Em_Mesh_Fruit}}}
    Spawn Rate=16
    Spawn Rate Per Particle Cap=64
    Use Maximum Age=false
    Use Minimum Age=false
   )
   &SpawnParticlesFromOtherEmitter(
    Attribute Reader=DataInterface() {propertyValues {EmitterBinding {bindingMode=Other, emitterName=Em_Mesh_Fruit}}}
    Spawn Group=1
    Spawn Rate=16
    Spawn Rate Per Particle Cap=64
    Use Maximum Age=false
    Use Minimum Age=false
   ) {module=SpawnParticlesFromOtherEmitter001}
  ]
  particleSpawnScript [
   &self:ScratchModule(Particle Attribute Reader=DataInterface() {propertyValues {EmitterBinding {bindingMode=Other, emitterName=Em_Mesh_Fruit}}}) {module=ScratchModule}
   &InitializeParticle(
    Color=&MaskLinearColorBySpawnGroup(
     Masked Linear Color=&SelectLinearColorFromArray(
      Array Sampling Mode=::NewEnumerator1
      Direct Array Index=$Particles.Partition
      Color Selection Array=DataInterface() {propertyValues {ColorData [<rgba 1 0.4334 0 1>, <rgba 0 0.2698 1 1>, <rgba 0.5693 0 1 1>, <rgba 1 0.1617 0.3607 1>]}}
      Random Seed=?
     )
     Passthrough Linear Color=<rgba 1 1 1 1>
     Spawn Group Mask=1
    )
    Lifetime=&MaskFloatBySpawnGroup(Masked Float=0.4, Passthrough Float=0.2, Spawn Group Mask=1)
    Mesh Scale Mode=::NewEnumerator3
    Mesh Uniform Scale=0.2
    Sprite Rotation Mode=::NewEnumerator2
    Sprite Size Mode=::NewEnumerator3
    Uniform Sprite Size=&RandomRangeFloat(Minimum=15, Maximum=22)
    UsePositionOffset=false
    Write Material Random=true
   )
   &SampleParticlesFromOtherEmitter(
    Maximum Sampled Particle Age:NiagaraFloat=&Lerp_Float(
     A=1.325
     B=0.425
     Alpha=&FloatFromCurve(
      FloatCurve=DataInterface() {propertyValues {Curve {keys [<at 0 1 RCIM_Constant>,<at 0.1 0 RCIM_Constant>]}}}
      Scale Curve=$User.ParaGlideLeanAlpha
     )
    )
    Minimum Sampled Particle Age:NiagaraFloat=&Lerp_Float(
     A=1.225
     B=0.325
     Alpha=&FloatFromCurve(
      FloatCurve=DataInterface() {propertyValues {Curve {keys [<at 0 1 RCIM_Constant>,<at 0.1 0 RCIM_Constant>]}}}
      Scale Curve=$User.ParaGlideLeanAlpha
     )
    )
    Module.MaxSampledAgeCutoff=true
    Module.MinSampledAgeCutoff=true
    Reject Particle By Sampled Age=true
   )
   &InitialMeshOrientation(Use Rotation Vector=false)
   &AddVelocity(
    Module.Use Falloff Distance=false
    Override Seed=false
    Velocity Mode=::NewEnumerator1
    Velocity Speed=&MaskFloatBySpawnGroup(Masked Float=75, Passthrough Float=0, Spawn Group Mask=1)
   )
   /SetVariables(Particles.VisibilityTag=$Emitter.SpawnGroup) {isSetParameters=true}
  ]
  particleUpdateScript [
   &ParticleState()
   &GravityForce(Gravity=<0 0 -200>)
   &SolveForcesAndVelocity(Clamp Velocity=false, Limit Acceleration=false)
   &DynamicMaterialParameters(
    Index 0 Param 1=6
    Index 0 Param 2=4
    Index 0 Param 3=&FloatFromCurve(FloatCurve=DataInterface() {propertyValues {Curve {keys [
         {arriveTangent=-1.4258, interpMode=RCIM_Cubic, leaveTangent=-1.4258, tangentMode=RCTM_User, time=0, value=1}
         {tangentMode=RCTM_User, time=1, value=0}
        ]}}})
    Index 0 Param 4=&Multiply_Float(
     A=&FloatFromCurve(
      FloatCurve=DataInterface() {propertyValues {Curve {keys [{arriveTangent=-2.5373, interpMode=RCIM_Cubic, leaveTangent=-2.5373, tangentMode=RCTM_User, time=0, value=1}, <at 1 0>]}}}
      Scale Curve=10
     )
     B=5
    )
    Module.Index0Param0Write=true
    Module.Index0Param1Write=true
    Module.Index0Param2Write=true
    Module.Index0Param3Write=true
    Module.Index1Param0Write=true
    Module.Index1Param1Write=true
    Module.Index1Param2Write=true
    Module.Index1Param3Write=true
    Write Parameter Index 1=true
   )
   &ScaleSpriteSize(Uniform Curve Sprite Scale=DataInterface() {propertyValues {Curve {keys [
        {arriveTangent=8.6652, interpMode=RCIM_Cubic, leaveTangent=8.6652, tangentMode=RCTM_User, time=0, value=0}
        {arriveTangent=0.0375, interpMode=RCIM_Cubic, leaveTangent=0.0375, tangentMode=RCTM_User, time=1, value=3}
       ]}}})
   &ScaleColor(
    Scale Alpha=&Multiply_Float(
     A=&FloatFromCurve(FloatCurve=DataInterface() {propertyValues {Curve {keys [
          {arriveTangent=-0.1334, interpMode=RCIM_Cubic, leaveTangent=-0.1334, tangentMode=RCTM_User, time=0, value=1}
          {arriveTangent=-2.5256, tangentMode=RCTM_User, time=1, value=0}
         ]}}})
     B=1
    )
    Scale RGB=&MakeVectorFromLinearColorRGB(Color=&MultiplyLinearColor(
      Linear Color=<rgba 1 0.8622 0.5003 1>
      Scale Factor=&MakeVector4_FromLinearColor(LinearColor=$System.ToD Color)
     ))
    ScaleA=true
    ScaleRGB=true
   ) {module=ScaleColor002}
   /SetVariables(Particles.DropSizes:Vector2f=&Vector2DFromFloat(Value=1)) {isSetParameters=true}
   &ScaleMeshSize(
    Initial Mesh Scale:Vector3f=$Particles.Scale
    Scale Factor=&VectorFromFloat(Value=&FloatFromCurve(
      FloatCurve=DataInterface() {propertyValues {Curve {keys [
          {arriveTangent=-0.0357, interpMode=RCIM_Cubic, leaveTangent=-0.0357, tangentMode=RCTM_User, time=0, value=1}
          {arriveTangent=-14.6213, leaveTangent=-2, tangentMode=RCTM_User, time=0.05, value=0}
          <at 0.5 0>
          {arriveTangent=13.566, interpMode=RCIM_Cubic, leaveTangent=13.566, tangentMode=RCTM_User, time=0.95, value=0}
          {arriveTangent=0.1498, tangentMode=RCTM_User, time=1, value=1}
         ]}}}
      CurveIndex=&NormalizeDistanceRange(
       Start Position:NiagaraPosition=&ReturnCameraPosition(Coordinate Space:ENiagaraCoordinateSpace=::Simulation, Camera Query:NiagaraDataInterfaceCamera=DataInterface())
       End Position:NiagaraPosition=$Particles.Position
       Distance:NiagaraFloat=250
       Invert Normalized Range:NiagaraBool=true
      )
     ))
   ) {module=ScaleMeshSize001}
   &CameraOffset(Camera Offset Amount=-15)
   /Niagara/Modules/Update/Camera/ScaleAttributesByCameraDistance(Max Distance=1000, Max Scale=0.5, Min Scale=0.25) {enabled=false, module=ScaleAttributesByCameraDistance001}
  ]
  rendererClasses [&NiagaraMeshRendererProperties, &NiagaraSpriteRendererProperties]
  renderers [
   {rendererClass=&NiagaraMeshRendererProperties, rendererIndex=0}
   {rendererClass=&NiagaraSpriteRendererProperties, rendererIndex=1}
   {rendererClass=&NiagaraSpriteRendererProperties, rendererIndex=2}
   {rendererClass=&NiagaraMeshRendererProperties, rendererIndex=3}
   {rendererClass=&NiagaraMeshRendererProperties, rendererIndex=4}
  ]
 }
}
summary {
 systemName=NS_Contrail_TroutWrist
 emitters [
  {emitterName=Em_Mesh_Fruit, rendererClasses [&NiagaraMeshRendererProperties]}
  {emitterName=Em_Mesh_Fruit_Sliced, rendererClasses [&NiagaraMeshRendererProperties]}
  {emitterName=Em_Mesh_Swipe, rendererClasses [&NiagaraMeshRendererProperties, &NiagaraSpriteRendererProperties]}
  {emitterName=Em_Mesh_Fruit_Banana, rendererClasses [&NiagaraMeshRendererProperties]}
  {bEnabled=false, emitterName=Em_Banana_Splash, rendererClasses [&NiagaraSpriteRendererProperties]}
  {bEnabled=false, emitterName=Em_Banana_Drips, rendererClasses [&NiagaraSpriteRendererProperties]}
 ]
 userVariables [
  User.bIsFrontEnd {type=NiagaraBool}
  User.bIsFrontEndPreview {type=NiagaraBool}
  User.ParaGlideLeanAlpha {type=NiagaraFloat, defaultValue=0, description=""}
  User.RibbonIDOffset {type=NiagaraInt32}
 ]
}
